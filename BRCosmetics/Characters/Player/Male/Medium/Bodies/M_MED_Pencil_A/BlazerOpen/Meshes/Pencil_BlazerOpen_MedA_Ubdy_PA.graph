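# Physics asset for the upper body of the medium male Pencil open-blazer cosmetic, set up for secondary motion of jacket, shirt and tie.
# Pelvis, spine and thigh capsules form the body volume around the cloth bones.
# Two-link jacket chains hang from pelvis-linked MSTR bodies with tight angular limits, while tie and knot chains hang from spine_04.
# Master-to-body links and both shirt-front chains rotate freely; many jacket capsules retain puffhood shape names.
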PhysicsAsset /BRCosmetics/Characters/Player/Male/Medium/Bodies/M_MED_Pencil_A/BlazerOpen/Meshes/Pencil_BlazerOpen_MedA_Ubdy_PA
bodies {
 dyn_jacket_bk_01 [{shapeType=Capsule, length=10, radius=4.5, shapeName=dyn_puffhood_l_01_capsule, center=<4.5 0 3.5>, rotation=<rot -90 0 0>}]
 dyn_jacket_bk_01_MSTR [{shapeType=Capsule, length=0.505, radius=0.505, shapeName=dyn_jacket_bk_01_MSTR_capsule, rotation=<rot 0 -90 0>}]
 dyn_jacket_bk_02 [{shapeType=Capsule, length=10, radius=4.5, shapeName=dyn_puffhood_l_01_capsule, center=<4.5 0 3.5>, rotation=<rot -90 0 0>}]
 dyn_jacket_bk_l_01 [{shapeType=Capsule, length=10, radius=4.5, shapeName=dyn_puffhood_l_01_capsule, center=<4.5 0 4.5>, rotation=<rot -90 0 0>}]
 dyn_jacket_bk_l_01_MSTR [{shapeType=Capsule, length=0.505, radius=0.505, shapeName=dyn_jacket_bk_l_01_MSTR_capsule, rotation=<rot 0 -90 0>}]
 dyn_jacket_bk_l_02 [{shapeType=Capsule, length=10, radius=4.5, shapeName=dyn_puffhood_l_01_capsule, center=<4.5 0 4.5>, rotation=<rot -90 0 0>}]
 dyn_jacket_bk_r_01 [{shapeType=Capsule, length=10, radius=4.5, shapeName=dyn_puffhood_l_01_capsule, center=<4.5 0 4.5>, rotation=<rot -90 0 0>}]
 dyn_jacket_bk_r_01_MSTR [{shapeType=Capsule, length=0.505, radius=0.505, shapeName=dyn_jacket_bk_r_01_MSTR_capsule, rotation=<rot 0 -90 0>}]
 dyn_jacket_bk_r_02 [{shapeType=Capsule, length=10, radius=4.5, shapeName=dyn_puffhood_l_01_capsule, center=<4.5 0 4.5>, rotation=<rot -90 0 0>}]
 dyn_jacket_bkout_l_01 [{shapeType=Capsule, length=10, radius=4.5, shapeName=dyn_puffhood_l_01_capsule, center=<4.5 0 5>, rotation=<rot 90 0 0>}]
 dyn_jacket_bkout_l_01_MSTR [{shapeType=Capsule, length=0.505, radius=0.505, shapeName=dyn_jacket_bkout_l_01_MSTR_capsule, rotation=<rot 0 -90 0>}]
 dyn_jacket_bkout_l_02 [{shapeType=Capsule, length=10, radius=4.5, shapeName=dyn_puffhood_l_01_capsule, center=<4.5 0 5>, rotation=<rot 90 0 0>}]
 dyn_jacket_bkout_r_01 [{shapeType=Capsule, length=10, radius=4.5, shapeName=dyn_puffhood_l_01_capsule, center=<4.5 0 6>, rotation=<rot -90 0 0>}]
 dyn_jacket_bkout_r_01_MSTR [{shapeType=Capsule, length=0.505, radius=0.505, shapeName=dyn_jacket_bkout_r_01_MSTR_capsule, rotation=<rot 0 -90 0>}]
 dyn_jacket_bkout_r_02 [{shapeType=Capsule, length=10, radius=4.5, shapeName=dyn_puffhood_l_01_capsule, center=<4.5 0 6>, rotation=<rot -90 0 0>}]
 dyn_jacket_fr_l_01 [{shapeType=Capsule, length=10, radius=4.5, shapeName=dyn_jacket_fr_l_01_capsule, center=<4.5 -4 3>, rotation=<rot 90 0 0>}]
 dyn_jacket_fr_l_01_MSTR [{shapeType=Capsule, length=0.505, radius=0.505, shapeName=dyn_jacket_fr_l_01_MSTR_capsule, rotation=<rot 0 -90 0>}]
 dyn_jacket_fr_l_02 [{shapeType=Capsule, length=10, radius=4.5, shapeName=dyn_jacket_fr_l_02_capsule, center=<4 -4 3>, rotation=<rot 90 0 0>}]
 dyn_jacket_fr_r_01 [{shapeType=Capsule, length=10, radius=4.5, shapeName=dyn_jacket_fr_r_01_capsule, center=<4 2 4>, rotation=<rot 90 0 0>}]
 dyn_jacket_fr_r_01_MSTR [{shapeType=Capsule, length=0.505, radius=0.505, shapeName=dyn_jacket_fr_r_01_MSTR_capsule, rotation=<rot 0 -90 0>}]
 dyn_jacket_fr_r_02 [{shapeType=Capsule, length=10, radius=4.5, shapeName=dyn_jacket_fr_r_02_capsule, center=<4 2 4>, rotation=<rot 90 0 0>}]
 dyn_jacket_l_01 [{shapeType=Capsule, length=10, radius=4.5, shapeName=dyn_jacket_l_01_capsule, center=<4 0 4.25>, rotation=<rot 90 0 0>}]
 dyn_jacket_l_01_MSTR [{shapeType=Capsule, length=0.505, radius=0.505, shapeName=dyn_jacket_l_01_MSTR_capsule, rotation=<rot 0 -90 0>}]
 dyn_jacket_l_02 [{shapeType=Capsule, length=10, radius=4.5, shapeName=dyn_jacket_l_02_capsule, center=<4 0 4.25>, rotation=<rot 90 0 0>}]
 dyn_jacket_r_01 [{shapeType=Capsule, length=10, radius=4.5, shapeName=dyn_puffhood_l_01_capsule, center=<4.5 -2 4>, rotation=<rot -90 180 180>}]
 dyn_jacket_r_01_MSTR [{shapeType=Capsule, length=0.505, radius=0.505, shapeName=dyn_jacket_r_01_MSTR_capsule, rotation=<rot 0 -90 0>}]
 dyn_jacket_r_02 [{shapeType=Capsule, length=10, radius=4.5, shapeName=dyn_puffhood_l_01_capsule, center=<4.5 -2 4>, rotation=<rot -90 180 180>}]
 pelvis [
  {shapeType=Capsule, length=7, radius=11.75, shapeName=spine_01_capsule, center=<0 -1.5 0>}
  {shapeType=Capsule, length=7, radius=5, shapeName=pelvis_capsule, center=<0 6.5 0>}
 ]
 shirt_fr_l_dyn [{shapeType=Capsule, length=0.505, radius=0.505, shapeName=shirt_fr_l_dyn_capsule, rotation=<rot 0 -90 0>}]
 shirt_fr_l_mstr [{shapeType=Sphere, radius=0.505, shapeName=shirt_fr_l_mstr_sphere}]
 shirt_fr_r_dyn [{shapeType=Capsule, length=0.505, radius=0.505, shapeName=shirt_fr_r_dyn_capsule, rotation=<rot 0 -90 0>}]
 shirt_fr_r_mstr [{shapeType=Sphere, radius=0.505, shapeName=shirt_fr_r_mstr_sphere}]
 spine_01 [{shapeType=Capsule, length=7, radius=11, shapeName=spine_01_capsule}]
 spine_02 [{shapeType=Capsule, length=6, radius=12, shapeName=spine_02_capsule}]
 spine_03 [{shapeType=Capsule, length=4, radius=12, shapeName=spine_03_capsule}]
 spine_04 [{shapeType=Capsule, length=5, radius=12, shapeName=spine_04_capsule}]
 spine_05 [{shapeType=Capsule, length=12, radius=10, shapeName=spine_05_capsule}]
 thigh_l [{shapeType=Capsule, length=24.5, radius=6, shapeName=thigh_l_capsule, center=<-20 0 0>, rotation=<rot 90 0 0>}]
 thigh_r [{shapeType=Capsule, length=24.5, radius=6, shapeName=thigh_r_capsule, center=<20 0 0>, rotation=<rot 90 0 0>}]
 tie_01_dyn [{shapeType=Capsule, length=5, radius=3, shapeName=tie_01_dyn_capsule, center=<4 0 2.5>, rotation=<rot 90 0 0>}]
 tie_02_dyn [{shapeType=Capsule, length=12, radius=3, shapeName=tie_02_dyn_capsule, center=<10 0 2.5>, rotation=<rot 90 0 0>}]
 tie_knot_01_dyn [{shapeType=Sphere, radius=1, shapeName=tie_knot_01_dyn_sphere}]
 tie_knot_02_dyn [{shapeType=Capsule, length=4.54, radius=3.5, shapeName=tie_knot_02_dyn_capsule, center=<4 0 3.75>, rotation=<rot 90 0 0>}]
 tie_knot_mstr [{shapeType=Sphere, radius=1, shapeName=tie_knot_mstr_sphere}]
 tie_mstr [{shapeType=Capsule, length=0.505, radius=0.505, shapeName=tie_mstr_capsule, rotation=<rot 0 -90 0>}]
}
constraints [
 {bone1Name=spine_01, bone2Name=pelvis, swing1LimitDegrees=45, swing2LimitDegrees=45, twistLimitDegrees=45}
 {
  bone1Name=thigh_r
  bone2Name=pelvis
  swing1LimitDegrees=45
  swing1Motion=Free
  swing2LimitDegrees=45
  swing2Motion=Free
  twistLimitDegrees=45
  twistMotion=Free
 }
 {
  bone1Name=thigh_l
  bone2Name=pelvis
  swing1LimitDegrees=45
  swing1Motion=Free
  swing2LimitDegrees=45
  swing2Motion=Free
  twistLimitDegrees=45
  twistMotion=Free
 }
 {
  bone1Name=spine_02
  bone2Name=spine_01
  swing1LimitDegrees=45
  swing1Motion=Free
  swing2LimitDegrees=45
  swing2Motion=Free
  twistLimitDegrees=45
  twistMotion=Free
 }
 {
  bone1Name=spine_03
  bone2Name=spine_02
  swing1LimitDegrees=45
  swing1Motion=Free
  swing2LimitDegrees=45
  swing2Motion=Free
  twistLimitDegrees=45
  twistMotion=Free
 }
 {
  bone1Name=dyn_jacket_bk_01_MSTR
  bone2Name=pelvis
  swing1LimitDegrees=45
  swing1Motion=Free
  swing2LimitDegrees=45
  swing2Motion=Free
  twistLimitDegrees=45
  twistMotion=Free
 }
 {bone1Name=dyn_jacket_bk_01, bone2Name=dyn_jacket_bk_01_MSTR, swing1LimitDegrees=10, swing2LimitDegrees=30, twistLimitDegrees=5}
 {bone1Name=dyn_jacket_bk_02, bone2Name=dyn_jacket_bk_01, swing1LimitDegrees=10, swing2LimitDegrees=30, twistLimitDegrees=5}
 {
  bone1Name=dyn_jacket_bk_l_01_MSTR
  bone2Name=pelvis
  swing1LimitDegrees=45
  swing1Motion=Free
  swing2LimitDegrees=45
  swing2Motion=Free
  twistLimitDegrees=45
  twistMotion=Free
 }
 {bone1Name=dyn_jacket_bk_l_01, bone2Name=dyn_jacket_bk_l_01_MSTR, swing1LimitDegrees=10, swing2LimitDegrees=30, twistLimitDegrees=5}
 {bone1Name=dyn_jacket_bk_l_02, bone2Name=dyn_jacket_bk_l_01, swing1LimitDegrees=10, swing2LimitDegrees=30, twistLimitDegrees=5}
 {
  bone1Name=dyn_jacket_bk_r_01_MSTR
  bone2Name=pelvis
  swing1LimitDegrees=45
  swing1Motion=Free
  swing2LimitDegrees=45
  swing2Motion=Free
  twistLimitDegrees=45
  twistMotion=Free
 }
 {bone1Name=dyn_jacket_bk_r_01, bone2Name=dyn_jacket_bk_r_01_MSTR, swing1LimitDegrees=10, swing2LimitDegrees=30, twistLimitDegrees=5}
 {bone1Name=dyn_jacket_bk_r_02, bone2Name=dyn_jacket_bk_r_01, swing1LimitDegrees=10, swing2LimitDegrees=30, twistLimitDegrees=5}
 {
  bone1Name=dyn_jacket_bkout_l_01_MSTR
  bone2Name=pelvis
  swing1LimitDegrees=45
  swing1Motion=Free
  swing2LimitDegrees=45
  swing2Motion=Free
  twistLimitDegrees=45
  twistMotion=Free
 }
 {bone1Name=dyn_jacket_bkout_l_01, bone2Name=dyn_jacket_bkout_l_01_MSTR, swing1LimitDegrees=10, swing2LimitDegrees=30, twistLimitDegrees=5}
 {bone1Name=dyn_jacket_bkout_l_02, bone2Name=dyn_jacket_bkout_l_01, swing1LimitDegrees=10, swing2LimitDegrees=30, twistLimitDegrees=5}
 {
  bone1Name=dyn_jacket_bkout_r_01_MSTR
  bone2Name=pelvis
  swing1LimitDegrees=45
  swing1Motion=Free
  swing2LimitDegrees=45
  swing2Motion=Free
  twistLimitDegrees=45
  twistMotion=Free
 }
 {bone1Name=dyn_jacket_bkout_r_01, bone2Name=dyn_jacket_bkout_r_01_MSTR, swing1LimitDegrees=10, swing2LimitDegrees=30, twistLimitDegrees=5}
 {bone1Name=dyn_jacket_bkout_r_02, bone2Name=dyn_jacket_bkout_r_01, swing1LimitDegrees=10, swing2LimitDegrees=30, twistLimitDegrees=5}
 {
  bone1Name=dyn_jacket_fr_l_01_MSTR
  bone2Name=pelvis
  swing1LimitDegrees=45
  swing1Motion=Free
  swing2LimitDegrees=45
  swing2Motion=Free
  twistLimitDegrees=45
  twistMotion=Free
 }
 {
  bone1Name=dyn_jacket_fr_r_01_MSTR
  bone2Name=pelvis
  swing1LimitDegrees=45
  swing1Motion=Free
  swing2LimitDegrees=45
  swing2Motion=Free
  twistLimitDegrees=45
  twistMotion=Free
 }
 {
  bone1Name=dyn_jacket_l_01_MSTR
  bone2Name=pelvis
  swing1LimitDegrees=45
  swing1Motion=Free
  swing2LimitDegrees=45
  swing2Motion=Free
  twistLimitDegrees=45
  twistMotion=Free
 }
 {
  bone1Name=dyn_jacket_r_01_MSTR
  bone2Name=pelvis
  swing1LimitDegrees=45
  swing1Motion=Free
  swing2LimitDegrees=45
  swing2Motion=Free
  twistLimitDegrees=45
  twistMotion=Free
 }
 {bone1Name=dyn_jacket_r_01, bone2Name=dyn_jacket_r_01_MSTR, swing1LimitDegrees=10, swing2LimitDegrees=30, twistLimitDegrees=5}
 {bone1Name=dyn_jacket_r_02, bone2Name=dyn_jacket_r_01, swing1LimitDegrees=10, swing2LimitDegrees=30, twistLimitDegrees=5}
 {
  bone1Name=tie_mstr
  bone2Name=spine_04
  swing1LimitDegrees=45
  swing1Motion=Free
  swing2LimitDegrees=45
  swing2Motion=Free
  twistLimitDegrees=45
  twistMotion=Free
 }
 {
  bone1Name=shirt_fr_l_mstr
  bone2Name=pelvis
  swing1LimitDegrees=45
  swing1Motion=Free
  swing2LimitDegrees=45
  swing2Motion=Free
  twistLimitDegrees=45
  twistMotion=Free
 }
 {
  bone1Name=shirt_fr_l_dyn
  bone2Name=shirt_fr_l_mstr
  swing1LimitDegrees=45
  swing1Motion=Free
  swing2LimitDegrees=45
  swing2Motion=Free
  twistLimitDegrees=45
  twistMotion=Free
 }
 {
  bone1Name=shirt_fr_r_mstr
  bone2Name=pelvis
  swing1LimitDegrees=45
  swing1Motion=Free
  swing2LimitDegrees=45
  swing2Motion=Free
  twistLimitDegrees=45
  twistMotion=Free
 }
 {
  bone1Name=shirt_fr_r_dyn
  bone2Name=shirt_fr_r_mstr
  swing1LimitDegrees=45
  swing1Motion=Free
  swing2LimitDegrees=45
  swing2Motion=Free
  twistLimitDegrees=45
  twistMotion=Free
 }
 {
  bone1Name=tie_knot_mstr
  bone2Name=spine_04
  swing1LimitDegrees=45
  swing1Motion=Free
  swing2LimitDegrees=45
  swing2Motion=Free
  twistLimitDegrees=45
  twistMotion=Free
 }
 {
  bone1Name=tie_knot_01_dyn
  bone2Name=tie_knot_mstr
  swing1LimitDegrees=45
  swing1Motion=Free
  swing2LimitDegrees=45
  swing2Motion=Free
  twistLimitDegrees=45
  twistMotion=Free
 }
 {bone1Name=tie_knot_02_dyn, bone2Name=tie_knot_01_dyn, swing1LimitDegrees=30, swing2LimitDegrees=30, twistLimitDegrees=1}
 {
  bone1Name=spine_04
  bone2Name=spine_03
  swing1LimitDegrees=45
  swing1Motion=Free
  swing2LimitDegrees=45
  swing2Motion=Free
  twistLimitDegrees=45
  twistMotion=Free
 }
 {
  bone1Name=spine_05
  bone2Name=spine_04
  swing1LimitDegrees=45
  swing1Motion=Free
  swing2LimitDegrees=45
  swing2Motion=Free
  twistLimitDegrees=45
  twistMotion=Free
 }
 {bone1Name=dyn_jacket_l_01, bone2Name=dyn_jacket_l_01_MSTR, swing1LimitDegrees=10, swing2LimitDegrees=30, twistLimitDegrees=1}
 {bone1Name=dyn_jacket_l_02, bone2Name=dyn_jacket_l_01, swing1LimitDegrees=10, swing2LimitDegrees=30, twistLimitDegrees=1}
 {bone1Name=dyn_jacket_fr_l_01, bone2Name=dyn_jacket_fr_l_01_MSTR, swing1LimitDegrees=10, swing2LimitDegrees=30, twistLimitDegrees=5}
 {bone1Name=dyn_jacket_fr_l_02, bone2Name=dyn_jacket_fr_l_01, swing1LimitDegrees=10, swing2LimitDegrees=30, twistLimitDegrees=5}
 {bone1Name=dyn_jacket_fr_r_01, bone2Name=dyn_jacket_fr_r_01_MSTR, swing1LimitDegrees=10, swing2LimitDegrees=30, twistLimitDegrees=5}
 {bone1Name=dyn_jacket_fr_r_02, bone2Name=dyn_jacket_fr_r_01, swing1LimitDegrees=10, swing2LimitDegrees=30, twistLimitDegrees=5}
 {bone1Name=tie_01_dyn, bone2Name=tie_mstr, swing1LimitDegrees=30, swing2LimitDegrees=30, twistLimitDegrees=10}
 {bone1Name=tie_02_dyn, bone2Name=tie_01_dyn, swing1LimitDegrees=30, swing2LimitDegrees=30, twistLimitDegrees=10}
]
