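# Animation Blueprint for the open-blazer upper body of the male medium Pencil character, adding secondary motion to jacket, tie and emblem.
# It copies the parent pose through a Control Rig, then selects rigid-body physics or movement-weighted trail controllers.
# Skydive floating uses rigid body plus tie corrections.
# A final coat Control Rig reacts to crouch, jumping, physics state and lower-body cosmetic tags (slacks, shorts, athletic).
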AnimBlueprint /BRCosmetics/Characters/Player/Male/Medium/Bodies/M_MED_Pencil_A/BlazerOpen/Meshes/Pencil_BlazerOpen_MedA_Ubdy_AnimBP
target_skeleton=&/Pencil_BlazerOpen_MedA_Ubdy_Skeleton
graphs [
 AnimGraph {nodes [
   n1 = RigidBody!(ComponentPose=Local To Component(LocalPose=Use cached pose 'DeformPost'()), Alpha, OverrideWorldGravity:Vector="0,0,-980", ExternalForce) -> (Pose) {
    bClampLinearTranslationLimitToRefPose=true
    bOverrideWorldGravity=true
    overridePhysicsAsset=&/Pencil_BlazerOpen_MedA_Ubdy_PA
    simulationSpace=BaseBoneSpace
    alphaBoolBlend {blendOption=Linear}
    baseBoneRef {boneName=pelvis}
    componentAppliedLinearAccClamp=<3000 3000 3000>
    componentLinearAccScale=<0.1 0.1 0.3>
    componentLinearVelScale=<0.1 0.1 0.3>
    overrideWorldGravity {z=-980}
    simSpaceSettings {
     maxAngularAcceleration=1000
     maxAngularVelocity=3
     maxLinearAcceleration=700
     maxLinearVelocity=500
     velocityScaleZ=5
     worldAlpha=1
     externalLinearDragV=<0.2 0.2>
    }
   }
   n2 = Component To Local(ComponentPose=@n1)
   n3 = Get Bwd Movement() -> (Bwd Movement:double)
   n4 = Trail controller - Bone: dyn_jacket_bk_02!(ComponentPose=Local To Component(LocalPose=Use cached pose 'DeformPost'()), Alpha=@n3, FakeVelocity:Vector="0,0,0") -> (Pose) {
    bActorSpaceFakeVel=true
    bShowTrailLocation=true
    debugLifeTime=0.1
    maxDeltaTime=0.0333
    relaxationSpeedScale=3
    alphaScaleBiasClamp {bInterpResult=true}
    baseJoint {boneName=root}
    planarLimits [
     {drivingBone {boneName=root}, planeTransform {location=<0 0 10>, rotation=<rot 0 0 0>, scale=<1 1 1>}}
     {drivingBone {boneName=spine_03}, planeTransform {location=<0 -20 0>, rotation=<rot 0 15 -89.9999>, scale=<1 1 1>}}
    ]
    rotationLimits [{limitMax=<180 180 180>}, {limitMax=<180 180 180>}]
    rotationOffsets [<0 0 0>, <0 0 0>]
    trailBone {boneName=dyn_jacket_bk_02}
    trailRelaxationSpeed {editorCurveData {keys [{arriveTangent=-72.0739, interpMode=RCIM_Cubic, leaveTangent=-72.0739, tangentMode=RCTM_User, time=0, value=35}, <at 1 5 RCIM_Cubic>]}}
   }
   n5 = Trail controller - Bone: dyn_jacket_bk_l_02!(ComponentPose=@n4, Alpha=@n3, FakeVelocity:Vector="0,0,0") -> (Pose) {
    bActorSpaceFakeVel=true
    bShowTrailLocation=true
    debugLifeTime=0.1
    maxDeltaTime=0.0333
    relaxationSpeedScale=3
    alphaScaleBiasClamp {bInterpResult=true}
    baseJoint {boneName=root}
    planarLimits [
     {drivingBone {boneName=root}, planeTransform {location=<0 0 10>, rotation=<rot 0 0 0>, scale=<1 1 1>}}
     {drivingBone {boneName=spine_03}, planeTransform {location=<0 -20 0>, rotation=<rot 0 15 -89.9999>, scale=<1 1 1>}}
    ]
    rotationLimits [{limitMax=<180 180 180>}, {limitMax=<180 180 180>}]
    rotationOffsets [<0 0 0>, <0 0 0>]
    trailBone {boneName=dyn_jacket_bk_l_02}
    trailRelaxationSpeed {editorCurveData {keys [{arriveTangent=-72.0739, interpMode=RCIM_Cubic, leaveTangent=-72.0739, tangentMode=RCTM_User, time=0, value=35}, <at 1 5 RCIM_Cubic>]}}
   }
   n6 = Trail controller - Bone: dyn_jacket_bk_r_02!(ComponentPose=@n5, Alpha=@n3, FakeVelocity:Vector="0,0,0") -> (Pose) {
    bActorSpaceFakeVel=true
    bShowTrailLocation=true
    debugLifeTime=0.1
    maxDeltaTime=0.0333
    relaxationSpeedScale=3
    alphaScaleBiasClamp {bInterpResult=true}
    baseJoint {boneName=root}
    planarLimits [
     {drivingBone {boneName=root}, planeTransform {location=<0 0 10>, rotation=<rot 0 0 0>, scale=<1 1 1>}}
     {drivingBone {boneName=spine_03}, planeTransform {location=<0 -20 0>, rotation=<rot 0 15 -89.9999>, scale=<1 1 1>}}
    ]
    rotationLimits [{limitMax=<180 180 180>}, {limitMax=<180 180 180>}]
    rotationOffsets [<0 0 0>, <0 0 0>]
    trailBone {boneName=dyn_jacket_bk_r_02}
    trailRelaxationSpeed {editorCurveData {keys [{arriveTangent=-72.0739, interpMode=RCIM_Cubic, leaveTangent=-72.0739, tangentMode=RCTM_User, time=0, value=35}, <at 1 5 RCIM_Cubic>]}}
   }
   n7 = Get Right Movement() -> (Right Movement:double)
   n8 = Trail controller - Bone: dyn_jacket_bkout_r_02!(
    ComponentPose=@n6
    Alpha=FInterpEaseInOut!(A:double=@n3, B:double=@n7, Alpha:double=1, Exponent:double=0) -> (ReturnValue:double)
    FakeVelocity:Vector="0,0,0"
   ) -> (Pose) {
    bActorSpaceFakeVel=true
    bShowTrailLocation=true
    debugLifeTime=0.1
    maxDeltaTime=0.0333
    relaxationSpeedScale=3
    alphaScaleBiasClamp {bInterpResult=true}
    baseJoint {boneName=root}
    planarLimits [
     {drivingBone {boneName=root}, planeTransform {location=<0 0 10>, rotation=<rot 0 0 0>, scale=<1 1 1>}}
     {drivingBone {boneName=spine_03}, planeTransform {location=<0 -20 0>, rotation=<rot 0 15 -89.9999>, scale=<1 1 1>}}
    ]
    rotationLimits [{limitMax=<180 180 180>}, {limitMax=<180 180 180>}]
    rotationOffsets [<0 0 0>, <0 0 0>]
    trailBone {boneName=dyn_jacket_bkout_r_02}
    trailRelaxationSpeed {editorCurveData {keys [{arriveTangent=-72.0739, interpMode=RCIM_Cubic, leaveTangent=-72.0739, tangentMode=RCTM_User, time=0, value=35}, <at 1 5 RCIM_Cubic>]}}
   }
   n9 = Get Left Movement() -> (Left Movement:double)
   n10 = Trail controller - Bone: dyn_jacket_bkout_l_02!(ComponentPose=@n8, Alpha=FInterpEaseInOut(A=@n3, B=@n9, Alpha=1, Exponent=0), FakeVelocity:Vector="0,0,0") -> (Pose) {
    bActorSpaceFakeVel=true
    bShowTrailLocation=true
    debugLifeTime=0.1
    maxDeltaTime=0.0333
    relaxationSpeedScale=3
    alphaScaleBiasClamp {bInterpResult=true}
    baseJoint {boneName=root}
    planarLimits [
     {drivingBone {boneName=root}, planeTransform {location=<0 0 10>, rotation=<rot 0 0 0>, scale=<1 1 1>}}
     {drivingBone {boneName=spine_03}, planeTransform {location=<0 -20 0>, rotation=<rot 0 15 -89.9999>, scale=<1 1 1>}}
    ]
    rotationLimits [{limitMax=<180 180 180>}, {limitMax=<180 180 180>}]
    rotationOffsets [<0 0 0>, <0 0 0>]
    trailBone {boneName=dyn_jacket_bkout_l_02}
    trailRelaxationSpeed {editorCurveData {keys [{arriveTangent=-72.0739, interpMode=RCIM_Cubic, leaveTangent=-72.0739, tangentMode=RCTM_User, time=0, value=35}, <at 1 5 RCIM_Cubic>]}}
   }
   n11 = Trail controller - Bone: dyn_jacket_l_02!(ComponentPose=@n10, Alpha=@n9, FakeVelocity:Vector="0,0,0") -> (Pose) {
    bActorSpaceFakeVel=true
    bShowTrailLocation=true
    debugLifeTime=0.1
    maxDeltaTime=0.0333
    relaxationSpeedScale=3
    alphaScaleBiasClamp {bInterpResult=true}
    baseJoint {boneName=root}
    planarLimits [
     {drivingBone {boneName=root}, planeTransform {location=<0 0 10>, rotation=<rot 0 0 0>, scale=<1 1 1>}}
     {drivingBone {boneName=spine_03}, planeTransform {location=<0 -20 0>, rotation=<rot 0 15 -89.9999>, scale=<1 1 1>}}
    ]
    rotationLimits [{limitMax=<180 180 180>}, {limitMax=<180 180 180>}]
    rotationOffsets [<0 0 0>, <0 0 0>]
    trailBone {boneName=dyn_jacket_l_02}
    trailRelaxationSpeed {editorCurveData {keys [{arriveTangent=-72.0739, interpMode=RCIM_Cubic, leaveTangent=-72.0739, tangentMode=RCTM_User, time=0, value=35}, <at 1 5 RCIM_Cubic>]}}
   }
   n12 = Trail controller - Bone: dyn_jacket_fr_l_02!(ComponentPose=@n11, Alpha=@n9, FakeVelocity:Vector="0,0,0") -> (Pose) {
    bActorSpaceFakeVel=true
    bShowTrailLocation=true
    debugLifeTime=0.1
    maxDeltaTime=0.0333
    relaxationSpeedScale=3
    alphaScaleBiasClamp {bInterpResult=true}
    baseJoint {boneName=root}
    planarLimits [
     {drivingBone {boneName=root}, planeTransform {location=<0 0 10>, rotation=<rot 0 0 0>, scale=<1 1 1>}}
     {drivingBone {boneName=spine_03}, planeTransform {location=<0 -20 0>, rotation=<rot 0 15 -89.9999>, scale=<1 1 1>}}
    ]
    rotationLimits [{limitMax=<180 180 180>}, {limitMax=<180 180 180>}]
    rotationOffsets [<0 0 0>, <0 0 0>]
    trailBone {boneName=dyn_jacket_fr_l_02}
    trailRelaxationSpeed {editorCurveData {keys [{arriveTangent=-72.0739, interpMode=RCIM_Cubic, leaveTangent=-72.0739, tangentMode=RCTM_User, time=0, value=35}, <at 1 5 RCIM_Cubic>]}}
   }
   n13 = Trail controller - Bone: dyn_jacket_r_02!(ComponentPose=@n12, Alpha=@n7, FakeVelocity:Vector="0,0,0") -> (Pose) {
    bActorSpaceFakeVel=true
    bShowTrailLocation=true
    debugLifeTime=0.1
    maxDeltaTime=0.0333
    relaxationSpeedScale=3
    alphaScaleBiasClamp {bInterpResult=true}
    baseJoint {boneName=root}
    planarLimits [
     {drivingBone {boneName=root}, planeTransform {location=<0 0 10>, rotation=<rot 0 0 0>, scale=<1 1 1>}}
     {drivingBone {boneName=spine_03}, planeTransform {location=<0 -20 0>, rotation=<rot 0 15 -89.9999>, scale=<1 1 1>}}
    ]
    rotationLimits [{limitMax=<180 180 180>}, {limitMax=<180 180 180>}]
    rotationOffsets [<0 0 0>, <0 0 0>]
    trailBone {boneName=dyn_jacket_r_02}
    trailRelaxationSpeed {editorCurveData {keys [{arriveTangent=-72.0739, interpMode=RCIM_Cubic, leaveTangent=-72.0739, tangentMode=RCTM_User, time=0, value=35}, <at 1 5 RCIM_Cubic>]}}
   }
   n14 = Trail controller - Bone: dyn_jacket_fr_r_02!(ComponentPose=@n13, Alpha=@n7, FakeVelocity:Vector="0,0,0") -> (Pose) {
    bActorSpaceFakeVel=true
    bShowTrailLocation=true
    debugLifeTime=0.1
    maxDeltaTime=0.0333
    relaxationSpeedScale=3
    alphaScaleBiasClamp {bInterpResult=true}
    baseJoint {boneName=root}
    planarLimits [
     {drivingBone {boneName=root}, planeTransform {location=<0 0 10>, rotation=<rot 0 0 0>, scale=<1 1 1>}}
     {drivingBone {boneName=spine_03}, planeTransform {location=<0 -20 0>, rotation=<rot 0 15 -89.9999>, scale=<1 1 1>}}
    ]
    rotationLimits [{limitMax=<180 180 180>}, {limitMax=<180 180 180>}]
    rotationOffsets [<0 0 0>, <0 0 0>]
    trailBone {boneName=dyn_jacket_fr_r_02}
    trailRelaxationSpeed {editorCurveData {keys [{arriveTangent=-72.0739, interpMode=RCIM_Cubic, leaveTangent=-72.0739, tangentMode=RCTM_User, time=0, value=35}, <at 1 5 RCIM_Cubic>]}}
   }
   n15 = Component To Local(ComponentPose=@n14)
   n16 = Get bRigidBodyEnabled()
   n17 = RigidBody(ComponentPose=Local To Component(LocalPose=Use cached pose 'DeformPost'()), OverrideWorldGravity="0,0,-980") {
    bClampLinearTranslationLimitToRefPose=true
    bOverrideWorldGravity=true
    overridePhysicsAsset=&Pencil_BlazerOpen_MedA_Ubdy_PA
    simulationSpace=BaseBoneSpace
    alphaBoolBlend {blendOption=Linear}
    baseBoneRef {boneName=pelvis}
    componentAppliedLinearAccClamp=<3000 3000 3000>
    componentLinearAccScale=<0.1 0.1 0.3>
    componentLinearVelScale=<0.1 0.1 0.3>
    overrideWorldGravity {z=-980}
    simSpaceSettings {
     maxAngularAcceleration=1000
     maxAngularVelocity=3
     maxLinearAcceleration=700
     maxLinearVelocity=500
     velocityScaleZ=5
     worldAlpha=1
     externalLinearDragV=<0.2 0.2>
    }
   }
   n18 = "Transform (Modify) Bone - Bone: tie_02_dyn"(ComponentPose=@n17, Alpha=1, Translation="0,0,1", Rotation="15,0,0", Scale="1,1,1") {
    rotationMode=BMM_Additive
    rotationSpace=BCS_BoneSpace
    translationMode=BMM_Additive
    translationSpace=BCS_BoneSpace
    alphaBoolBlend {blendOption=Linear}
    boneToModify {boneName=tie_02_dyn}
    rotation {pitch=15}
    translation {z=1}
   }
   n19 = "Constraint - tie_01_dyn (PPPPPPP)"!(
    ComponentPose=@n18
    Alpha=0.5
    ConstraintWeights_0=0.002
    ConstraintWeights_1:float=0.0257
    ConstraintWeights_2:float=0.1577
    ConstraintWeights_3:float=0.2691
    ConstraintWeights_4:float=0.407
    ConstraintWeights_5:float=0.1385
    ConstraintWeights_6:float=0
   ) -> (Pose) {
    alphaBoolBlend {blendOption=Linear}
    boneToModify {boneName=tie_01_dyn}
    constraintSetup [
     {targetBone {boneName=pelvis}}
     {targetBone {boneName=spine_01}}
     {targetBone {boneName=spine_02}}
     {targetBone {boneName=spine_03}}
     {targetBone {boneName=spine_04}}
     {targetBone {boneName=spine_05}}
     {targetBone {boneName=clavicle_l}}
    ]
    constraintWeights [1,1,1,1,1,1,1]
   }
   n20 = "Constraint - tie_02_dyn (PPPPP)"!(
    ComponentPose=@n19
    Alpha=0.5
    ConstraintWeights_0=0.0527
    ConstraintWeights_1:float=0.1247
    ConstraintWeights_2:float=0.3194
    ConstraintWeights_3:float=0.3345
    ConstraintWeights_4:float=0.1686
   ) -> (Pose) {
    alphaBoolBlend {blendOption=Linear}
    boneToModify {boneName=tie_02_dyn}
    constraintSetup [
     {targetBone {boneName=pelvis}}
     {targetBone {boneName=spine_01}}
     {targetBone {boneName=spine_02}}
     {targetBone {boneName=spine_03}}
     {targetBone {boneName=spine_04}}
    ]
    constraintWeights [1,1,1,1,1]
   }
   n21 = Control Rig!(
    Source=Blend Poses by bool(
     BlendPose_0=Component To Local(ComponentPose=@n20)
     BlendPose_1=Blend Poses by bool(BlendPose_0=@n2, BlendPose_1=@n15, bActiveValue=@n16) {blendType=Linear}
     bActiveValue=Get bIsSkydiveFloating()
    ) {blendType=Linear}
    Alpha
    Crouch_CTRL:bool=Get bIsCrouching()
    Pencil_Slacks:bool=Get Pencil_Slacks() -> (Pencil_Slacks:bool)
    Pencil_Athletic:bool=Get Pencil_Athletic() -> (Pencil_Athletic:bool)
    Pencil_Shorts:bool=Get Pencil_Shorts() -> (Pencil_Shorts:bool)
    RBAN Disabled:bool=NOT Boolean(A=Get bRigidBodyEnabled())
    RBAN Enabled:bool=Get bRigidBodyEnabled()
    Jumping:bool=OR Boolean(A=Get bIsJumping(), B=Get bIsFalling())
   ) -> (Pose) {
    alphaBoolBlend {blendOption=Linear}
    controlRigAssetReference {blueprintRigClass=&/Pencil_BlazerOpen_MedA_Ubdy_Coat_ControlRig.Pencil_BlazerOpen_MedA_Ubdy_Coat_ControlRig_C}
   }
   Output Pose(Result=@n21) {layerGroup=None}
   n22 = Control Rig!(Source=Copy Pose From Mesh() {bCopyCurves=true, bUseAttachedParent=true}, Alpha) -> (Pose) {
    alphaBoolBlend {blendOption=Linear}
    controlRigAssetReference {blueprintRigClass=&/Pencil_BlazerOpen_MedA_Ubdy_ControlRig.Pencil_BlazerOpen_MedA_Ubdy_ControlRig_C}
   }
   n23 = "Constraint - emblem_mstr (PPPPP)"!(
    ComponentPose=Local To Component(LocalPose=@n22)
    Alpha=1
    ConstraintWeights_0=0.0565
    ConstraintWeights_1:float=0.2176
    ConstraintWeights_2:float=0.3349
    ConstraintWeights_3:float=0.3096
    ConstraintWeights_4:float=0.0814
   ) -> (Pose) {
    alphaBoolBlend {blendOption=Linear}
    boneToModify {boneName=emblem_mstr}
    constraintSetup [
     {targetBone {boneName=spine_03}}
     {targetBone {boneName=spine_04}}
     {targetBone {boneName=spine_05}}
     {targetBone {boneName=clavicle_l}}
     {targetBone {boneName=deform_upperarm_fr_l}}
    ]
    constraintWeights [1,1,1,1,1]
   }
   Save cached pose 'DeformPost'(Pose=Component To Local(ComponentPose=@n23)) {cacheName=DeformPost}
  ]}
 EventGraph {nodes [
   n1 = PlayerMovement(AnimInstance=Self-Reference(), XDotProductCoefficient=1, YDotProductCoefficient=1, ZDotProductCoefficient=1)
   n2 = MapRangeClamped(Value=@n1.YMovement, InRangeA=0, InRangeB=-1)
   n3 = float * float(A="Absolute (Float)"(A=@n1), B=0.5)
   > Set Bwd Movement(Bwd Movement:double=MapRangeClamped(
     Value=float + float(A=@n3, B=@n2) {is=title K2Node_CommutativeAssociativeBinaryOperator}
     InRangeA=0
     InRangeB=1.1
     OutRangeA=0.4
     OutRangeB=0
    )) -> (Output_Get:double)
   n4 = MapRangeClamped(Value=@n1, InRangeA=-0.2, InRangeB=1)
   n5 = float * float(A="Absolute (Float)"(A=@n1.YMovement), B=0.2)
   > Set Right Movement(Right Movement:double=MapRangeClamped(Value=float + float(A=@n5, B=@n4), InRangeA=0, InRangeB=1.2, OutRangeB=0.5)) -> (Output_Get:double)
   n6 = MapRangeClamped(Value=@n1, InRangeA=0.2, InRangeB=-1, OutRangeB=0.3)
   n7 = float * float(A="Absolute (Float)"(A=@n1.YMovement), B=0.2)
   > Set Left Movement(Left Movement:double=MapRangeClamped(Value=float + float(A=@n7, B=@n6), InRangeA=0, InRangeB=1.2, OutRangeB=0.5)) -> (Output_Get:double)
   Event BlueprintUpdateAnimation()
   > n8 = Sequence!(execute) -> (then_0, then_1, then_2:Exec, then_3:Exec)
   > n9 = Gravity Override State Function DEPRECATED(CharacterPartAnimInstance=Self-Reference(), GameTimeInSeconds=GetGameTimeInSeconds(), GravityOverrideParams=Get Gravity Override Params())
   > Set Gravity Override(Gravity Override=@n9)
   > Set Sky Dive Gravity Override(Sky Dive Gravity Override:Vector="0,0,-500") -> (Output_Get:Vector)
   n10 = ComplexSimSpaceFunction DEPRECATED!(
    execute=@n8.then_1
    CharacterPartAnimInstance:Custom Character Part Anim Instance Object Reference=Self-Reference()
    InterpSpeedMax:double=9
    PawnSpeedMax:double=20
    RInterpErrorTolerance:double=1
    DeltaTime:double=0
    DebugDraw:bool=false
    sim_space_rotation_in:Rotator=Get Sim Space Rotation() -> (Sim Space Rotation:Rotator)
    inComponentSpaceIn:bool=Get In Component Space() -> (In Component Space:bool)
    inComponentSpaceLastFrameIn:bool=Get In Component Space Last Frame() -> (In Component Space Last Frame:bool)
    lockedOnSimSpaceIn:bool=Get Locked on Sim Space() -> (Locked on Sim Space:bool)
    InterpSpeedIn:double=Get Interp Speed() -> (Interp Speed:double)
   ) -> (
    then
    sim_space_rotation:Rotator
    inComponentSpaceOut:bool
    inComponentSpaceLastFrameOut:bool
    lockedOnSimSpaceOut:bool
    InterpSpeedOut:double
   )
   > Set Sim Space Rotation(Sim Space Rotation:Rotator=@n10) -> (Output_Get:Rotator)
   n11 = Branch(execute=@n8.then_2, Condition=Get bIsSkyDiving())
   > n12 = Branch(Condition="float > float"(A=Get PawnVelocityZ(), B=50) {is=title K2Node_CallFunction})
   > Set Bwd Movement(Bwd Movement=float * float(A=Get Bwd Movement(), B=0))
   Set Bwd Movement(execute=[@n11.else, @n12.else], Bwd Movement=float * float(A=Get Bwd Movement(), B=1))
   n13 = PlayerMovement(execute=@n8.then_3, AnimInstance=Self-Reference(), XDotProductCoefficient=1, YDotProductCoefficient=1, ZDotProductCoefficient=1)
   n14 = MapRangeClamped(Value=@n13.YMovement, InRangeA=0, InRangeB=-1)
   n15 = float * float(A="Absolute (Float)"(A=@n13), B=0.5)
   > Set Bwd Movement(Bwd Movement=MapRangeClamped(Value=float + float(A=@n15, B=@n14), InRangeA=0, InRangeB=1.1, OutRangeA=0.1, OutRangeB=0))
   n16 = MapRangeClamped(Value=@n13, InRangeA=-0.2, InRangeB=1)
   n17 = float * float(A="Absolute (Float)"(A=@n13.YMovement), B=0.2)
   > Set Right Movement(Right Movement=MapRangeClamped(Value=float + float(A=@n17, B=@n16), InRangeA=0, InRangeB=1.2, OutRangeB=0))
   n18 = MapRangeClamped(Value=@n13, InRangeA=0.2, InRangeB=-1, OutRangeB=0.3)
   n19 = float * float(A="Absolute (Float)"(A=@n13.YMovement), B=0.2)
   > Set Left Movement(Left Movement=MapRangeClamped(Value=float + float(A=@n19, B=@n18), InRangeA=0, InRangeB=1.2, OutRangeB=0.5))
   Event OnCharacterCustomizationLoaded()
   > n20 = Sequence!(execute) -> (then_0, then_1, then_2)
   > n21 = Branch(Condition=CharacterItemDefinitionHasAnyMetaTag(Tags=Make Array("[0]"=(TagName="Cosmetics.MetaTags.Pencil.Lower.Slacks"))))
   > Set Pencil_Slacks(Pencil_Slacks:bool=true) -> (Output_Get:bool)
   Set Pencil_Slacks(execute=@n21.else, Pencil_Slacks=false)
   n22 = Branch(
    execute=@n20.then_1
    Condition=CharacterItemDefinitionHasAnyMetaTag(Tags=Make Array("[0]"=(TagName="Cosmetics.MetaTags.Pencil.Lower.Shorts")))
   )
   > Set Pencil_Shorts(Pencil_Shorts:bool=true) -> (Output_Get:bool)
   Set Pencil_Shorts(execute=@n22.else, Pencil_Shorts=false)
   n23 = Branch(
    execute=@n20.then_2
    Condition=CharacterItemDefinitionHasAnyMetaTag(Tags=Make Array("[0]"=(TagName="Cosmetics.MetaTags.Pencil.Lower.Athletic")))
   )
   > Set Pencil_Athletic(Pencil_Athletic:bool=true) -> (Output_Get:bool)
   Set Pencil_Athletic(execute=@n23.else, Pencil_Athletic=false)
   MapRangeClamped(Value=0, InRangeA=0, InRangeB=1.2, OutRangeB=0.5)
   Set Interp Speed(Interp Speed:double=@n10.InterpSpeedOut) -> (Output_Get:double)
   Set Locked on Sim Space(Locked on Sim Space:bool=@n10.lockedOnSimSpaceOut) -> (Output_Get:bool)
   Set In Component Space Last Frame(In Component Space Last Frame:bool=@n10.inComponentSpaceLastFrameOut) -> (Output_Get:bool)
   Set In Component Space(In Component Space:bool=@n10.inComponentSpaceOut) -> (Output_Get:bool)
  ]}
]
variables [
 Bwd Movement {type=number}
 Crouching {type=boolean}
 Gravity Override {type=Vector}
 Gravity Override Params {type=GravityOverrideParamsStruct}
 In Component Space {type=boolean}
 In Component Space Last Frame {type=boolean}
 Interp Speed {type=number}
 Left Movement {type=number}
 Locked on Sim Space {type=boolean}
 Pencil_Athletic {type=boolean}
 Pencil_Shorts {type=boolean}
 Pencil_Slacks {type=boolean, default=true}
 Right Movement {type=number}
 Sim Space Rotation {type=Rotator}
 Sky Dive Gravity Override {type=Vector}
]
